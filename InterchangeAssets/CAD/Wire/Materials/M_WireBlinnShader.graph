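# Default Lit parent material that approximates a Blinn shader for CAD Wire assets, exposing numbered Blinn parameters.
# Color blends toward a slightly specular-tinted color by Gloss, scales by Diffuse, adds incandescence, dims by inverse transparency, then multiplies by amplified GlowIntensity.
# SpecularRolloff's square root blends a constant toward a Fresnel term, scaled by Reflectivity.
# A roughness-like value inverts the product of Specularity and inverse Eccentricity.
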Material /InterchangeAssets/CAD/Wire/Materials/M_WireBlinnShader
graph {
 properties {usedShadingModels=Default Lit}
 nodes [
  n1 = VectorParameter() {parameterName="Color (8)", defaultValue=<rgba 0.0497 0.4287 0.0356 1>}
  n2 = VectorParameter() {parameterName="SpecularColor (3)", defaultValue=<rgba 0.0194 0.0194 0.0185 1>}
  n3 = Constant() {r=0.96}
  n4 = Multiply(
   A=LinearInterpolate(A=@n1.RGB, B=LinearInterpolate(A=@n2.RGB, B=@n1.RGB, Alpha=@n3), Alpha=ScalarParameter() {defaultValue=0.8, parameterName="Gloss (2)"})
   B=LinearInterpolate(A=Constant() {r=0.04}, B=Constant() {r=1}, Alpha=ScalarParameter() {defaultValue=1, parameterName="Diffuse (1)"})
  )
  n5 = VectorParameter() {parameterName="IncandescenceColor (9)", defaultValue=<rgba 0 0 0 1>}
  n6 = VectorParameter() {parameterName="TransparencyColor (10)", defaultValue=<rgba 0 0 0 1>}
  Multiply(
   A=Multiply(A=Add(A=@n4, B=@n5.RGB), B=OneMinus(None=@n6.RGB))
   B=Multiply(A=ScalarParameter() {parameterName="GlowIntensity (11)"}, B=Constant() {r=100})
  )
  n7 = Constant() {r=4}
  n8 = ScalarParameter() {defaultValue=0.65, parameterName="SpecularRolloff (5)"}
  n9 = Constant() {r=0.5}
  Multiply(
   A=LinearInterpolate(
    A=Constant() {r=1}
    B=MaterialFunctionCall(Power=@n7) {materialFunction=&Fresnel_Function}.Result
    Alpha=Power(Base=@n8, Exp=@n9)
   )
   B=ScalarParameter() {defaultValue=0.95, parameterName="Reflectivity (7)"}
  )
  n10 = ScalarParameter() {defaultValue=0.55, parameterName="Eccentricity (6)"}
  OneMinus(None=Multiply(A=OneMinus(None=@n10), B=ScalarParameter() {defaultValue=0.3, parameterName="Specularity (4)"}))
 ]
}
