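# Default Lit material that models a Lambert shader for Wire CAD imports in Interchange, with parameters for color, diffuse, incandescence, transparency and glow.
# Diffuse scales Color from a small floor up to full, then IncandescenceColor is added.
# The sum is multiplied by inverted TransparencyColor, so transparency darkens rather than sets opacity.
# A final multiply applies heavily amplified GlowIntensity; no connection to material outputs is shown.
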Material /InterchangeAssets/CAD/Wire/Materials/M_WireLambertShader
graph {
 properties {usedShadingModels=Default Lit}
 nodes [
  n1 = Multiply(
   A=VectorParameter() {parameterName="Color (2)", defaultValue=<rgba 0.0409 0.5711 0.6514 1>}.RGB
   B=LinearInterpolate(A=Constant() {r=0.04}, B=Constant() {r=1}, Alpha=ScalarParameter() {defaultValue=1, parameterName="Diffuse (1)"})
  )
  n2 = VectorParameter() {parameterName="IncandescenceColor (3)", defaultValue=<rgba 0 0 0 1>}
  n3 = VectorParameter() {parameterName="TransparencyColor (4)", defaultValue=<rgba 0 0 0 1>}
  Multiply(
   A=Multiply(A=Add(A=@n1, B=@n2.RGB), B=OneMinus(None=@n3.RGB))
   B=Multiply(A=ScalarParameter() {parameterName="GlowIntensity (5)"}, B=Constant() {r=100})
  )
 ]
}
