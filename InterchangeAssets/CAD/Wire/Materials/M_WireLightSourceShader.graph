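# Material for light source shaders in the Interchange CAD Wire assets, computing a glow color from exposed color, incandescence, transparency and glow intensity parameters.
# Color and IncandescenceColor are added, then attenuated by the inverse of TransparencyColor.
# The result is multiplied by GlowIntensity, which is scaled by a fixed factor of 100.
# Parameter names carry numeric suffixes, and the shading model is Default Lit.
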Material /InterchangeAssets/CAD/Wire/Materials/M_WireLightSourceShader
graph {
 properties {usedShadingModels=Default Lit}
 nodes [
  n1 = VectorParameter() {parameterName="Color (1)", defaultValue=<rgba 0.9647 0.9647 0.9647 1>}
  n2 = VectorParameter() {parameterName="IncandescenceColor (2)", defaultValue=<rgba 0 0 0 1>}
  n3 = VectorParameter() {parameterName="TransparencyColor (3)", defaultValue=<rgba 0 0 0 1>}
  Multiply(
   A=Multiply(A=Add(A=@n1.RGB, B=@n2.RGB), B=OneMinus(None=@n3.RGB))
   B=Multiply(A=ScalarParameter() {parameterName="GlowIntensity (4)"}, B=Constant() {r=100})
  )
 ]
}
