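# M_WireLightSourceTransparentShader is a translucent, surface-lit material in the Interchange CAD Wire assets, representing a transparent light-source shader.
# The sum of color and incandescence parameters is multiplied by inverted transparency color and scaled glow intensity, forming a glow term.
# The inverted transparency is also reduced to a scalar, apparently for opacity, by a three-way average that counts red twice and omits green.
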Material /InterchangeAssets/CAD/Wire/Materials/M_WireLightSourceTransparentShader
graph {
 properties {blendMode=BLEND_Translucent, translucencyLightingMode=TLM_Surface, usedShadingModels=Default Lit}
 nodes [
  n1 = OneMinus(None=VectorParameter() {parameterName="TransparencyColor (3)", defaultValue=<rgba 0.2159 0.2159 0.2159 1>}.RGB)
  n2 = MaterialFunctionCall(Float3=@n1) {materialFunction=&BreakOutFloat3Components}
  Divide(A=Add(A=Add(A=@n2.R, B=@n2.R), B=@n2.B), B=Constant() {r=3})
  n3 = VectorParameter() {parameterName="Color (1)", defaultValue=<rgba 0.9647 0.9647 0.9647 1>}
  n4 = VectorParameter() {parameterName="IncandescenceColor (2)", defaultValue=<rgba 0 0 0 1>}
  Multiply(A=Multiply(A=Add(A=@n3.RGB, B=@n4.RGB), B=@n1), B=Multiply(A=ScalarParameter() {parameterName="GlowIntensity (4)"}, B=Constant() {r=100}))
 ]
}
